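# Animation Blueprint for the ApplePound backpack cosmetic, fitting the pack to the character and animating its chains.
# EventGraph samples body arm and pelvis sockets, spine bend and movement state, and toggles a hide material parameter for the matching pickaxe.
# AnimGraph copies those sockets onto proxy bones, then blends front-end, crouch, standing and skydive/parachute bone offsets.
# Constraints, Anim Dynamics and Look At nodes then drive secondary chain motion.
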AnimBlueprint /BRCosmetics/Accessories/FORT_Backpacks/Backpack_M_MED_ApplePound/Meshes/M_MED_ApplePound_Pack_AnimBP
target_skeleton=&/M_MED_ApplePound_Pack_Skeleton
graphs [
 AnimGraph {nodes [
   "Transform (Modify) Bone - Bone: mstr_base_dn_rotate"!(
    ComponentPose
    bAlphaBoolEnabled:bool=false
    Translation=Get closing location() -> (closing location:Vector)
    Rotation="0,0,0"
    Scale="1,1,1"
   ) -> (Pose) {
    alphaInputType=Bool
    rotationMode=BMM_Additive
    scaleMode=BMM_Additive
    scaleSpace=BCS_BoneSpace
    translationMode=BMM_Additive
    alphaBoolBlend {blendInTime=0.3, blendOption=CircularInOut, blendOutTime=0.8}
    boneToModify {boneName=mstr_base_dn_rotate}
   }
   n1 = "Transform (Modify) Bone - Bone: mstr_base_up_rotate"!(
    ComponentPose=Local To Component(LocalPose=Use cached pose 'INPUT'())
    bAlphaBoolEnabled:bool=True
    Translation="0,3,0"
    Rotation="0,0,6"
    Scale="1,1,1"
   ) -> (Pose) {
    alphaInputType=Bool
    rotationMode=BMM_Additive
    scaleMode=BMM_Additive
    scaleSpace=BCS_BoneSpace
    translationMode=BMM_Additive
    alphaBoolBlend {blendInTime=0.2, blendOption=Linear, blendOutTime=0.8}
    boneToModify {boneName=mstr_base_up_rotate}
    rotation {roll=6}
    translation {y=3}
   }
   n2 = "Transform (Modify) Bone - Bone: pack"(
    ComponentPose=@n1
    Alpha=Get Updates Backpack Crouch Alpha() -> (Updates Backpack Crouch Alpha:double)
    Translation=Get backpack_translation()
    Rotation=Get backpack_rotate()
   ) {
    rotationMode=BMM_Additive
    scaleMode=BMM_Additive
    scaleSpace=BCS_BoneSpace
    translationMode=BMM_Additive
    alphaBoolBlend {blendOption=Linear}
    boneToModify {boneName=pack}
   }
   n3 = "Constraint - base (T)"(ComponentPose=@n2, Alpha=0.2, ConstraintWeights_0=1) {
    alphaBoolBlend {blendOption=Linear}
    boneToModify {boneName=base}
    constraintSetup [{transformType=Translation, targetBone {boneName=upperarm_left}}]
    constraintWeights [1]
   }
   n4 = "Transform (Modify) Bone - Bone: mstr_base_up_rotate"(
    ComponentPose=Local To Component(LocalPose=Use cached pose 'INPUT'())
    bAlphaBoolEnabled=True
    Translation="0,5,0"
    Rotation="0,0,8"
    Scale="1,1,1"
   ) {
    alphaInputType=Bool
    rotationMode=BMM_Additive
    scaleMode=BMM_Additive
    scaleSpace=BCS_BoneSpace
    translationMode=BMM_Additive
    alphaBoolBlend {blendInTime=0.2, blendOption=Linear, blendOutTime=0.8}
    boneToModify {boneName=mstr_base_up_rotate}
    rotation {roll=8}
    translation {y=5}
   }
   n5 = "Transform (Modify) Bone - Bone: mstr_base_up_rotate"(ComponentPose=@n4, bAlphaBoolEnabled=True, Translation="0,0,0", Rotation="0,0,6", Scale="1,1,1") {
    alphaInputType=Bool
    rotationMode=BMM_Additive
    scaleMode=BMM_Additive
    scaleSpace=BCS_BoneSpace
    translationMode=BMM_Additive
    alphaBoolBlend {blendInTime=0.3, blendOption=CircularIn, blendOutTime=0.3}
    boneToModify {boneName=mstr_base_up_rotate}
    rotation {roll=6}
   }
   n6 = "Constraint - base (T)"(ComponentPose=Local To Component(LocalPose=Use cached pose 'INPUT'()), Alpha=0.6, ConstraintWeights_0=1) {
    alphaBoolBlend {blendOption=Linear}
    boneToModify {boneName=base}
    constraintSetup [{transformType=Translation, targetBone {boneName=upperarm_left}}]
    constraintWeights [1]
   }
   n7 = "Transform (Modify) Bone - Bone: base"!(ComponentPose=@n6, bAlphaBoolEnabled:bool=True, Translation="4,2,3", Rotation="0,0,0", Scale="0.95,0.95,0.92") -> (Pose) {
    alphaInputType=Bool
    rotationMode=BMM_Additive
    scaleMode=BMM_Additive
    scaleSpace=BCS_BoneSpace
    translationMode=BMM_Additive
    translationSpace=BCS_BoneSpace
    alphaBoolBlend {blendInTime=0.3, blendOption=Sinusoidal, blendOutTime=0.8}
    boneToModify {boneName=base}
    scale=<0.95 0.95 0.92>
    translation=<4 2 3>
   }
   n8 = "Constraint - mstr_base_dn_rotate (T)"(ComponentPose=@n7, Alpha=1, ConstraintWeights_0=1) {
    alphaBoolBlend {blendOption=Linear}
    boneToModify {boneName=mstr_base_dn_rotate}
    constraintSetup [{transformType=Translation, targetBone {boneName=hips}}]
    constraintWeights [1]
   }
   n9 = "Transform (Modify) Bone - Bone: mstr_base_dn_rotate"(ComponentPose=@n8, bAlphaBoolEnabled=True, Translation=Get spine_interp() -> (spine_interp:Vector), Rotation="0,0,0", Scale="1,1,1") {
    alphaInputType=Bool
    rotationMode=BMM_Additive
    scaleMode=BMM_Additive
    scaleSpace=BCS_BoneSpace
    translationMode=BMM_Additive
    translationSpace=BCS_BoneSpace
    alphaBoolBlend {blendInTime=0.4, blendOption=CircularOut, blendOutTime=0.6}
    boneToModify {boneName=mstr_base_dn_rotate}
   }
   n10 = "Transform (Modify) Bone - Bone: mstr_base_up_rotate"!(
    ComponentPose=@n9
    Alpha:float=Get SkyDivingModeAlpha() -> (SkyDivingModeAlpha:double)
    Translation="0,2.5,0"
    Rotation="0,0,0"
    Scale="1,1,1"
   ) -> (Pose) {
    rotationMode=BMM_Additive
    scaleMode=BMM_Additive
    scaleSpace=BCS_BoneSpace
    translationMode=BMM_Additive
    alphaBoolBlend {blendInTime=0.2, blendOption=Linear, blendOutTime=0.8}
    boneToModify {boneName=mstr_base_up_rotate}
    translation {y=2.5}
   }
   n11 = "Transform (Modify) Bone - Bone: mstr_base_up_rotate"(ComponentPose=@n10, Alpha=Get Start Skydiving() -> (Start Skydiving:double), Translation="0,-2.5,0", Rotation="0,0,0", Scale="1,1,1") {
    rotationMode=BMM_Additive
    scaleMode=BMM_Additive
    scaleSpace=BCS_BoneSpace
    translationMode=BMM_Additive
    alphaBoolBlend {blendInTime=0.2, blendOption=Linear, blendOutTime=0.8}
    boneToModify {boneName=mstr_base_up_rotate}
    translation {y=-2.5}
   }
   n12 = "Transform (Modify) Bone - Bone: base"(
    ComponentPose=@n11
    bAlphaBoolEnabled=Get RightMovementSkydiving() -> (RightMovementSkydiving:bool)
    Translation="0,0,0"
    Rotation="0,0,0"
    Scale="1,1,1"
   ) {
    alphaInputType=Bool
    rotationMode=BMM_Additive
    scaleMode=BMM_Additive
    scaleSpace=BCS_BoneSpace
    translationMode=BMM_Additive
    translationSpace=BCS_BoneSpace
    alphaBoolBlend {blendInTime=0.3, blendOption=CircularOut, blendOutTime=0.8}
    alphaScaleBias {bias=-0.6, scale=2}
    boneToModify {boneName=base}
   }
   n13 = "Transform (Modify) Bone - Bone: mstr_base_dn_rotate"(ComponentPose=@n12, bAlphaBoolEnabled=Get RightMovementSkydiving(), Translation="0,0,0", Rotation="0,0,0", Scale="1,1,1") {
    alphaInputType=Bool
    rotationMode=BMM_Additive
    scaleMode=BMM_Additive
    scaleSpace=BCS_BoneSpace
    translationMode=BMM_Additive
    alphaBoolBlend {blendInTime=0.3, blendOption=Sinusoidal, blendOutTime=0.8}
    boneToModify {boneName=mstr_base_dn_rotate}
   }
   n14 = "Transform (Modify) Bone - Bone: base"(
    ComponentPose=@n13
    bAlphaBoolEnabled=Get LeftMovementSkydiving() -> (LeftMovementSkydiving:bool)
    Translation="0,0,0"
    Rotation="0,0,0"
    Scale="1,1,1"
   ) {
    alphaInputType=Bool
    rotationMode=BMM_Additive
    scaleMode=BMM_Additive
    scaleSpace=BCS_BoneSpace
    translationMode=BMM_Additive
    translationSpace=BCS_BoneSpace
    alphaBoolBlend {blendInTime=0.3, blendOption=Sinusoidal, blendOutTime=0.8}
    boneToModify {boneName=base}
   }
   n15 = "Transform (Modify) Bone - Bone: mstr_base_dn_rotate"(
    ComponentPose=@n14
    bAlphaBoolEnabled=Get LeftMovementSkydiving()
    Translation=Get moving left() -> (moving left:Vector)
    Rotation="0,0,0"
    Scale="1,1,1"
   ) {
    alphaInputType=Bool
    rotationMode=BMM_Additive
    scaleMode=BMM_Additive
    scaleSpace=BCS_BoneSpace
    translationMode=BMM_Additive
    alphaBoolBlend {blendInTime=0.3, blendOption=Sinusoidal, blendOutTime=0.8}
    boneToModify {boneName=mstr_base_dn_rotate}
   }
   n16 = "Transform (Modify) Bone - Bone: mstr_base_dn_rotate"(ComponentPose=@n15, bAlphaBoolEnabled=True, Translation="0,8,0", Rotation="0,0,0", Scale="1,1,1") {
    alphaInputType=Bool
    rotationMode=BMM_Additive
    scaleMode=BMM_Additive
    scaleSpace=BCS_BoneSpace
    translationMode=BMM_Additive
    translationSpace=BCS_BoneSpace
    alphaBoolBlend {blendInTime=0.3, blendOption=Sinusoidal, blendOutTime=0.8}
    boneToModify {boneName=mstr_base_dn_rotate}
    translation {y=8}
   }
   n17 = "Transform (Modify) Bone - Bone: mstr_base_dn_rotate"(
    ComponentPose=@n16
    bAlphaBoolEnabled=Get bIsSkydiveFloating()
    Translation=Get sky floating() -> (sky floating:Vector)
    Rotation="0,0,0"
    Scale="1,1,1"
   ) {
    alphaInputType=Bool
    rotationMode=BMM_Additive
    scaleMode=BMM_Additive
    scaleSpace=BCS_BoneSpace
    translationMode=BMM_Additive
    translationSpace=BCS_BoneSpace
    alphaBoolBlend {blendInTime=0.3, blendOption=Sinusoidal, blendOutTime=0.8}
    boneToModify {boneName=mstr_base_dn_rotate}
   }
   n18 = Look At - Bone: base_rotate_up(ComponentPose=@n17, Alpha=1) {
    bUseLookUpAxis=true
    alphaBoolBlend {blendOption=Linear}
    boneToModify {boneName=base_rotate_up}
    lookAtLocation {x=0}
    lookAtTarget {boneReference {boneName=base_rotate_dn}}
    lookAt_Axis {axis=<1 0>}
    lookUp_Axis {axis {x=0, z=1}}
   }
   n19 = "Transform (Modify) Bone - Bone: mstr_base"!(ComponentPose=@n18, bAlphaBoolEnabled:bool=Get bIsParachuteOpened(), Translation="0,-2,0", Rotation="0,0,0", Scale="1,1,1") -> (Pose) {
    alphaInputType=Bool
    rotationMode=BMM_Additive
    scaleMode=BMM_Additive
    scaleSpace=BCS_BoneSpace
    translationMode=BMM_Additive
    alphaBoolBlend {blendInTime=0.4, blendOption=CircularIn, blendOutTime=0.7}
    boneToModify {boneName=mstr_base}
    translation {y=-2}
   }
   n20 = "Transform (Modify) Bone - Bone: mstr_base_up_rotate"!(
    ComponentPose=@n19
    bAlphaBoolEnabled:bool=Get parachute opening() -> (parachute opening:bool)
    Translation="-2,0,-1"
    Rotation="0,0,0"
    Scale="1,1,1"
   ) -> (Pose) {
    alphaInputType=Bool
    rotationMode=BMM_Additive
    scaleMode=BMM_Additive
    scaleSpace=BCS_BoneSpace
    translationMode=BMM_Additive
    alphaBoolBlend {blendInTime=0.3, blendOption=Sinusoidal, blendOutTime=0.8}
    boneToModify {boneName=mstr_base_up_rotate}
    translation {x=-2, z=-1}
   }
   n21 = "Transform (Modify) Bone - Bone: mstr_base_up_rotate"(
    ComponentPose=@n20
    bAlphaBoolEnabled=Get parachute closing() -> (parachute closing:bool)
    Translation="-2,0,-1"
    Rotation="0,0,0"
    Scale="1,1,1"
   ) {
    alphaInputType=Bool
    rotationMode=BMM_Additive
    scaleMode=BMM_Additive
    scaleSpace=BCS_BoneSpace
    translationMode=BMM_Additive
    alphaBoolBlend {blendInTime=0.3, blendOption=Sinusoidal, blendOutTime=0.8}
    boneToModify {boneName=mstr_base_up_rotate}
    translation {x=-2, z=-1}
   }
   n22 = "Transform (Modify) Bone - Bone: mstr_base"(ComponentPose=@n21, bAlphaBoolEnabled=True, Translation="0,1,0", Rotation="0,0,0", Scale="1,1,1") {
    alphaInputType=Bool
    rotationMode=BMM_Additive
    scaleMode=BMM_Additive
    scaleSpace=BCS_BoneSpace
    translationMode=BMM_Additive
    alphaBoolBlend {blendInTime=0.2, blendOption=Linear, blendOutTime=0.8}
    boneToModify {boneName=mstr_base}
    translation {y=1}
   }
   n23 = Blend Poses by bool(
    BlendPose_0=Component To Local(ComponentPose=@n22)
    BlendPose_1=Blend Poses by bool(
     BlendPose_0=Component To Local(ComponentPose=@n5)
     BlendPose_1=Component To Local(ComponentPose=@n3)
     BlendTime_0=0.3
     BlendTime_1=0.3
     bActiveValue=Get bIsCrouching()
    ) {blendType=Linear}
    BlendTime_0=0.3
    BlendTime_1=0.3
    bActiveValue=Get bIsSkyDiving()
   ) {blendType=Linear}
   n24 = "Transform (Modify) Bone - Bone: mstr_base"(
    ComponentPose=Local To Component(LocalPose=Input Pose() {name=InPose_0})
    bAlphaBoolEnabled=True
    Translation="0,1.2,0"
    Rotation="-0.077,-4.408,-0.450"
    Scale="1,1,1"
   ) {
    alphaInputType=Bool
    rotationMode=BMM_Additive
    scaleMode=BMM_Additive
    scaleSpace=BCS_BoneSpace
    translationMode=BMM_Additive
    alphaBoolBlend {blendInTime=0.2, blendOption=Linear, blendOutTime=0.8}
    boneToModify {boneName=mstr_base}
    rotation=<rot -0.077 -4.408 -0.45>
    translation {y=1.2}
   }
   n25 = Component To Local(ComponentPose=@n24)
   n26 = Get bIsInFrontEnd()
   n27 = "Constraint - mstr_chain_dn_B (P)"(
    ComponentPose=Local To Component(LocalPose=Blend Poses by bool(BlendPose_0=@n25, BlendPose_1=@n23, BlendTime_0=0.3, BlendTime_1=0.3, bActiveValue=@n26) {blendType=Linear})
    Alpha=1
    ConstraintWeights_0=1
   ) {
    alphaBoolBlend {blendOption=Linear}
    boneToModify {boneName=mstr_chain_dn_B}
    constraintSetup [{targetBone {boneName=base_rotate_up}}]
    constraintWeights [1]
   }
   n28 = "Constraint - mstr_chain_dn_A (P)"(ComponentPose=@n27, Alpha=1, ConstraintWeights_0=1) {
    alphaBoolBlend {blendOption=Linear}
    boneToModify {boneName=mstr_chain_dn_A}
    constraintSetup [{targetBone {boneName=base_rotate_up}}]
    constraintWeights [1]
   }
   n29 = "Constraint - mstr_chain_up_B (P)"(ComponentPose=@n28, Alpha=1, ConstraintWeights_0=1) {
    alphaBoolBlend {blendOption=Linear}
    boneToModify {boneName=mstr_chain_up_B}
    constraintSetup [{targetBone {boneName=base_rotate_up}}]
    constraintWeights [1]
   }
   n30 = "Constraint - mstr_target_chain_up (P)"(ComponentPose=@n29, Alpha=1, ConstraintWeights_0=1) {
    alphaBoolBlend {blendOption=Linear}
    boneToModify {boneName=mstr_target_chain_up}
    constraintSetup [{targetBone {boneName=base_rotate_up}}]
    constraintWeights [1]
   }
   n31 = "Transform (Modify) Bone - Bone: mstr_target_chain_up"(ComponentPose=@n30, Alpha=1, Translation="0.282,0.028,-0.025", Rotation="40.477,-65.482,-73.508", Scale="1,1,1") {
    rotationMode=BMM_Additive
    rotationSpace=BCS_BoneSpace
    translationMode=BMM_Additive
    translationSpace=BCS_BoneSpace
    alphaBoolBlend {blendOption=Linear}
    boneToModify {boneName=mstr_target_chain_up}
    rotation=<rot 40.477 -65.4821 -73.5081>
    translation=<0.282 0.028 -0.025>
   }
   n32 = "Transform (Modify) Bone - Bone: mstr_chain_up_D"(ComponentPose=@n31, Alpha=1, Translation="0,0,0", Rotation="0.261,-58.092,17.840", Scale="1,1,1") {
    rotationMode=BMM_Additive
    rotationSpace=BCS_BoneSpace
    translationMode=BMM_Additive
    translationSpace=BCS_BoneSpace
    alphaBoolBlend {blendOption=Linear}
    boneToModify {boneName=mstr_chain_up_D}
    rotation=<rot 0.261 -58.092 17.84>
   }
   n33 = "Transform (Modify) Bone - Bone: mstr_chain_up_B"(ComponentPose=@n32, Alpha=1, Translation="0.200,-0.072,-0.017", Rotation="43.545,-58.418,-59.425", Scale="1,1,1") {
    rotationMode=BMM_Additive
    rotationSpace=BCS_BoneSpace
    translationMode=BMM_Additive
    translationSpace=BCS_BoneSpace
    alphaBoolBlend {blendOption=Linear}
    boneToModify {boneName=mstr_chain_up_B}
    rotation=<rot 43.545 -58.418 -59.425>
    translation=<0.2 -0.072 -0.017>
   }
   n34 = Anim Dynamics - Bone: mstr_chain_up_B!(ComponentPose=@n33, Alpha=1, GravityOverride:Vector="0,0,-980", ExternalForce="0,0,0") -> (Pose) {
    angularDampingOverride=0.9999
    angularSpringConstant=1500
    bAngularSpring=true
    bOverrideAngularDamping=true
    bUseGravityOverride=true
    gravityScale=0.35
    alphaBoolBlend {blendOption=Linear}
    boundBone {boneName=mstr_chain_up_B}
    componentAppliedLinearAccClamp=<750 750 750>
    componentLinearAccScale=<0.5 0.5 0.5>
    componentLinearVelScale=<0.65 0.65 0.65>
    gravityOverride {z=-980}
    physicsBodyDefinitions [{
      boundBone {boneName=mstr_chain_up_B}
      boxExtents=<10 10 10>
      constraintSetup {angularLimitsMax=<0 0 0>, angularLimitsMin=<0 -60 0>, angularTarget=<1 0 0>}
      localJointOffset=<1 0 0>
     }]
   }
   n35 = Anim Dynamics - Bone: mstr_target_chain_up!(ComponentPose=@n34, Alpha=1, GravityOverride:Vector="0,0,-980", ExternalForce="0,0,0") -> (Pose) {
    angularDampingOverride=0.9999
    angularSpringConstant=1500
    bAngularSpring=true
    bOverrideAngularDamping=true
    bUseGravityOverride=true
    gravityScale=0.35
    simulationSpace=RootRelative
    alphaBoolBlend {blendOption=Linear}
    boundBone {boneName=mstr_target_chain_up}
    componentAppliedLinearAccClamp=<750 750 750>
    componentLinearAccScale=<0.5 0.5 0.5>
    componentLinearVelScale=<0.65 0.65 0.65>
    gravityOverride {z=-980}
    physicsBodyDefinitions [{
      boundBone {boneName=mstr_target_chain_up}
      boxExtents=<10 10 10>
      constraintSetup {angularLimitsMax=<0 0 0>, angularLimitsMin=<0 -60 0>, angularTarget=<1 0 0>}
      localJointOffset=<1 0 0>
     }]
   }
   n36 = Anim Dynamics - Bone: mstr_chain_up_D!(ComponentPose=@n35, Alpha=0, GravityOverride:Vector="0,0,-980", ExternalForce="0,0,0") -> (Pose) {
    angularDampingOverride=0.9999
    angularSpringConstant=1500
    bAngularSpring=true
    bOverrideAngularDamping=true
    bUseGravityOverride=true
    gravityScale=0.35
    simulationSpace=RootRelative
    alphaBoolBlend {blendOption=Linear}
    boundBone {boneName=mstr_chain_up_D}
    componentAppliedLinearAccClamp=<1000 1000 1000>
    componentLinearAccScale=<0.45 0.45 0.45>
    componentLinearVelScale=<0.65 0.65 0.65>
    gravityOverride {z=-980}
    physicsBodyDefinitions [{
      boundBone {boneName=mstr_chain_up_D}
      boxExtents=<10 10 10>
      constraintSetup {angularLimitsMax=<0 0 0>, angularLimitsMin=<0 0 0>, angularTarget=<1 0 0>}
      localJointOffset=<1 0 0>
     }]
   }
   n37 = Anim Dynamics - Bone: mstr_chain_up_D!(ComponentPose=@n36, Alpha=0.6, ExternalForce="0,0,0") -> (Pose) {
    angularDampingOverride=1
    bOverrideAngularDamping=true
    bOverrideLinearDamping=true
    bUseGravityOverride=true
    linearDampingOverride=1
    linearSpringConstant=3000
    alphaBoolBlend {blendOption=Linear}
    boundBone {boneName=mstr_chain_up_D}
    componentAppliedLinearAccClamp=<2000 2000 2000>
    componentLinearAccScale=<0.8 0.8 0.8>
    componentLinearVelScale=<0.95 0.95 0.95>
    physicsBodyDefinitions [{
      boundBone {boneName=mstr_chain_up_D}
      boxExtents=<60 60 60>
      constraintSetup {angularLimitsMax=<0 0 0>, angularLimitsMin=<0 0 0>, angularTarget=<0 0 0>, linearAxesMax=<0 0.1 0.2>, linearAxesMin=<-0.65 -0.2 -0.1>}
      localJointOffset=<0 0 0>
     }]
   }
   n38 = Anim Dynamics - Bone: dyn_chain_dn_D(ComponentPose=@n37, Alpha=0.6, ExternalForce="0,0,0") {
    angularDampingOverride=1
    bOverrideAngularDamping=true
    bOverrideLinearDamping=true
    bUseGravityOverride=true
    linearDampingOverride=1
    linearSpringConstant=3000
    alphaBoolBlend {blendOption=Linear}
    boundBone {boneName=dyn_chain_dn_D}
    componentAppliedLinearAccClamp=<2000 2000 2000>
    componentLinearAccScale=<0.8 0.8 0.8>
    componentLinearVelScale=<0.95 0.95 0.95>
    physicsBodyDefinitions [{
      boundBone {boneName=dyn_chain_dn_D}
      boxExtents=<60 60 60>
      constraintSetup {angularLimitsMax=<0 0 0>, angularLimitsMin=<0 0 0>, angularTarget=<0 0 0>, linearAxesMax=<0 0 0.1>, linearAxesMin=<-0.8 0 -0.2>}
      localJointOffset=<0 0 0>
     }]
   }
   n39 = "Constraint - mstr_chain_up (TT)"!(ComponentPose=@n38, Alpha=1, ConstraintWeights_0=0.5, ConstraintWeights_1:float=0.5) -> (Pose) {
    alphaBoolBlend {blendOption=Linear}
    boneToModify {boneName=mstr_chain_up}
    constraintSetup [{transformType=Translation, targetBone {boneName=mstr_chain_up_F}}, {transformType=Translation, targetBone {boneName=mstr_chain_up_C}}]
    constraintWeights [1,1]
   }
   n40 = Anim Dynamics - Bone: dyn_chain_dn_A!(ComponentPose=@n39, Alpha=1, GravityOverride:Vector="0,0,-980", ExternalForce="0,0,0") -> (Pose) {
    angularDampingOverride=0.9999
    angularSpringConstant=1500
    bAngularSpring=true
    bOverrideAngularDamping=true
    bUseGravityOverride=true
    bUsePlanarLimit=false
    gravityScale=0.35
    alphaBoolBlend {blendOption=Linear}
    boundBone {boneName=dyn_chain_dn_A}
    componentAppliedLinearAccClamp=<1000 1000 1000>
    componentLinearAccScale=<0.85 0.85 0.85>
    componentLinearVelScale=<0.9 0.9 0.9>
    gravityOverride {z=-980}
    physicsBodyDefinitions [{
      boundBone {boneName=dyn_chain_dn_A}
      boxExtents=<12 12 12>
      constraintSetup {angularLimitsMax=<30 30 30>, angularLimitsMin=<30 -30 -30>, angularTarget=<1 0 0>}
      localJointOffset=<1 0 0>
     }]
   }
   n41 = Anim Dynamics - Bone: dyn_chain_dn_D!(ComponentPose=@n40, Alpha=1, GravityOverride:Vector="0,0,-980", ExternalForce="0,0,0") -> (Pose) {
    angularBiasOverride=1
    angularDampingOverride=1
    angularSpringConstant=1500
    bAngularSpring=true
    bOverrideAngularBias=true
    bOverrideAngularDamping=true
    bUseGravityOverride=true
    gravityScale=0.35
    alphaBoolBlend {blendOption=Linear}
    boundBone {boneName=dyn_chain_dn_D}
    componentAppliedLinearAccClamp=<1000 1000 1000>
    componentLinearAccScale=<0.85 0.85 0.85>
    componentLinearVelScale=<0.9 0.9 0.9>
    gravityOverride {z=-980}
    physicsBodyDefinitions [{
      boundBone {boneName=dyn_chain_dn_D}
      boxExtents=<12 12 12>
      constraintSetup {angularLimitsMax=<0 45 0>, angularLimitsMin=<0 -45 0>, angularTarget=<1 0 0>}
      localJointOffset=<1 0 0>
     }]
   }
   n42 = Anim Dynamics - Bone: dyn_chain_dn_B!(ComponentPose=@n41, Alpha=1, GravityOverride:Vector="0,0,-980", ExternalForce="0,0,0") -> (Pose) {
    angularDampingOverride=0.9999
    angularSpringConstant=1500
    bOverrideAngularDamping=true
    bUseGravityOverride=true
    gravityScale=0.35
    alphaBoolBlend {blendOption=Linear}
    boundBone {boneName=dyn_chain_dn_B}
    componentAppliedLinearAccClamp=<750 750 750>
    componentLinearAccScale=<0.5 0.5 0.5>
    componentLinearVelScale=<0.6 0.6 0.6>
    gravityOverride {z=-980}
    physicsBodyDefinitions [{
      boundBone {boneName=dyn_chain_dn_B}
      boxExtents=<7 7 7>
      constraintSetup {angularLimitsMax=<10 30 10>, angularLimitsMin=<-15 -30 -10>, angularTarget=<1 0 0>}
      localJointOffset=<1 0 0>
     }]
   }
   n43 = "Constraint - mstr_chain_dn_C (TT)"!(ComponentPose=@n42, Alpha=1, ConstraintWeights_0=0.7, ConstraintWeights_1:float=0.3) -> (Pose) {
    alphaBoolBlend {blendOption=Linear}
    boneToModify {boneName=mstr_chain_dn_C}
    constraintSetup [{transformType=Translation, targetBone {boneName=target_chain_dn}}, {transformType=Translation, targetBone {boneName=dyn_chain_dn_B}}]
    constraintWeights [1,1]
   }
   n44 = "Transform (Modify) Bone - Bone: target_chain_dn"(ComponentPose=@n43, Alpha=1, Translation="0,0,0", Rotation="0,0,0", Scale="1,1,1") {translationMode=BMM_Additive, translationSpace=BCS_BoneSpace, alphaBoolBlend {blendOption=Linear}, boneToModify {boneName=target_chain_dn}}
   n45 = Anim Dynamics - Bone: mstr_chain_dn_C(ComponentPose=@n44, Alpha=0.6, ExternalForce="0,0,0") {
    angularDampingOverride=1
    bOverrideAngularDamping=true
    bOverrideLinearDamping=true
    bUseGravityOverride=true
    linearDampingOverride=1
    linearSpringConstant=3000
    alphaBoolBlend {blendOption=Linear}
    boundBone {boneName=mstr_chain_dn_C}
    componentAppliedLinearAccClamp=<2000 2000 2000>
    componentLinearAccScale=<0.8 0.8 0.8>
    componentLinearVelScale=<0.95 0.95 0.95>
    physicsBodyDefinitions [{
      boundBone {boneName=mstr_chain_dn_C}
      boxExtents=<60 60 60>
      constraintSetup {angularLimitsMax=<0 0 0>, angularLimitsMin=<0 0 0>, angularTarget=<0 0 0>, linearAxesMax=<0 0.1 0.2>, linearAxesMin=<-0.65 -0.2 -0.1>}
      localJointOffset=<0 0 0>
     }]
   }
   n46 = Anim Dynamics - Bone: dyn_chain_dn_D!(ComponentPose=@n45, Alpha=0.6, ExternalForce="0,0,0") -> (Pose) {
    angularDampingOverride=1
    bOverrideAngularDamping=true
    bOverrideLinearDamping=true
    bUseGravityOverride=true
    linearDampingOverride=1
    linearSpringConstant=3000
    alphaBoolBlend {blendOption=Linear}
    boundBone {boneName=dyn_chain_dn_D}
    componentAppliedLinearAccClamp=<2000 2000 2000>
    componentLinearAccScale=<0.8 0.8 0.8>
    componentLinearVelScale=<0.95 0.95 0.95>
    physicsBodyDefinitions [{
      boundBone {boneName=dyn_chain_dn_D}
      boxExtents=<60 60 60>
      constraintSetup {angularLimitsMax=<0 0 0>, angularLimitsMin=<0 0 0>, angularTarget=<0 0 0>, linearAxesMax=<0 0 0.1>, linearAxesMin=<-0.8 0 -0.2>}
      localJointOffset=<0 0 0>
     }]
   }
   n47 = Look At - Bone: effector_chain_dn_C(ComponentPose=@n46, Alpha=1) {
    bUseLookUpAxis=true
    alphaBoolBlend {blendOption=Linear}
    boneToModify {boneName=effector_chain_dn_C}
    lookAtLocation {x=0}
    lookAtTarget {boneReference {boneName=target_chain_dn}}
    lookAt_Axis {axis=<1 0>}
    lookUp_Axis {axis=<0 1>}
   }
   n48 = Look At - Bone: dyn_chain_dn_B(ComponentPose=@n47, Alpha=1) {
    bUseLookUpAxis=true
    alphaBoolBlend {blendOption=Linear}
    boneToModify {boneName=dyn_chain_dn_B}
    lookAtLocation {x=0}
    lookAtTarget {boneReference {boneName=mstr_chain_dn_C}}
    lookAt_Axis {axis=<1 0>}
    lookUp_Axis {axis=<0 1>}
   }
   Output Pose(Result=Component To Local(ComponentPose=@n48)) {layerGroup=None}
   n49 = "Transform (Modify) Bone - Bone: upperarm_left"(
    ComponentPose=Local To Component(LocalPose=Input Pose())
    Alpha=1
    Translation=Get upperarm_l translation() -> (upperarm_l translation:Vector)
    Rotation=Get upperarm_l rotation() -> (upperarm_l rotation:Rotator)
    Scale="1,1,1"
   ) {
    rotationMode=BMM_Replace
    rotationSpace=BCS_WorldSpace
    scaleMode=BMM_Additive
    translationMode=BMM_Replace
    translationSpace=BCS_WorldSpace
    alphaBoolBlend {blendInTime=1, blendOption=Linear, blendOutTime=1}
    boneToModify {boneName=upperarm_left}
   }
   n50 = "Transform (Modify) Bone - Bone: hips"(
    ComponentPose=@n49
    Alpha=1
    Translation=Get Location_pelvis() -> (Location_pelvis:Vector)
    Rotation=Get Rotation_Pelvis() -> (Rotation_Pelvis:Rotator)
    Scale="1,1,1"
   ) {
    rotationMode=BMM_Replace
    rotationSpace=BCS_WorldSpace
    scaleMode=BMM_Additive
    translationMode=BMM_Replace
    translationSpace=BCS_WorldSpace
    alphaBoolBlend {blendInTime=1, blendOption=Linear, blendOutTime=1}
    boneToModify {boneName=hips}
   }
   Save cached pose 'INPUT'(Pose=Component To Local(ComponentPose=@n50)) {cacheName=INPUT}
   Get parachute closing()
  ]}
 EventGraph {nodes [
   Event OnParachuteClosed()
   > Set parachute closed(parachute closed:bool=true) -> (Output_Get:bool)
   > Delay(Duration=3)
   > Set parachute closed(parachute closed=false)
   n1 = Event BlueprintUpdateAnimation()
   > Set Delta Time X(Delta Time X=@n1.DeltaTimeX)
   > n2 = Sequence!(execute) -> (then_0, then_1, then_2:Exec, then_3:Exec, then_4:Exec, then_5:Exec, then_6:Exec, then_7:Exec, then_8:Exec)
   n3 = Sequence!(execute=@n2.then_2) -> (then_0, then_1, then_2)
   n4 = TryGetPawnOwner()
   n5 = Dot Product!(
    A:Vector=GetVelocity!(self:Actor Object Reference=@n4) -> (ReturnValue:Vector)
    B:Vector=GetActorRightVector!(self:Actor Object Reference=@n4) -> (ReturnValue:Vector)
   ) -> (ReturnValue:double)
   > n6 = Set RightMovementAlphaSkydiving(RightMovementAlphaSkydiving:double=MapRangeClamped(Value=@n5, InRangeA=0, InRangeB=230)) -> (Output_Get:double)
   > Set RightMovementSkydiving(RightMovementSkydiving:bool="Equal (Float)"(A=@n6, B=1)) -> (Output_Get:bool)
   > n7 = Set LeftMovementAlphaSkydiving(LeftMovementAlphaSkydiving:double=MapRangeClamped(Value=@n5, InRangeA=0, InRangeB=-230)) -> (Output_Get:double)
   > Set LeftMovementSkydiving(LeftMovementSkydiving:bool="Equal (Float)"(A=@n7, B=1)) -> (Output_Get:bool)
   n8 = Sequence(execute=@n3.then_1)
   n9 = Sequence!(execute=@n8.then_2) -> (then_0, then_1)
   > Set parachute closing(parachute closing:bool="Equal (Float)"(A=Get End OP() -> (End OP:double), B=1)) -> (Output_Get:bool)
   Set parachute opening(execute=@n9.then_1, parachute opening:bool="Equal (Float)"(A=Get Start OP() -> (Start OP:double), B=1)) -> (Output_Get:bool)
   n10 = Branch(execute=@n8.then_1, Condition=Get parachute closed() -> (parachute closed:bool))
   > Do Once(Reset=@n11.then)
   > Set End OP(End OP:double=1) -> (Output_Get:double)
   > Delay(Duration=0.3)
   > Set End OP(End OP=0)
   n11 = Set End OP(execute=@n10.else, End OP=0)
   n12 = Branch(execute=@n8.then_0, Condition=Get bIsParachuteOpened())
   > Do Once(Reset=@n13.then)
   > Set Start OP(Start OP:double=1) -> (Output_Get:double)
   > Delay(Duration=2)
   > Set Start OP(Start OP=0)
   n13 = Set Start OP(execute=@n12.else, Start OP=0)
   n14 = Gravity Override State Function DEPRECATED(
    execute=@n2.then_3
    CharacterPartAnimInstance=Self-Reference()
    GameTimeInSeconds=GetGameTimeInSeconds()
    GravityOverrideParams=Get Gravity Override Params()
   )
   > Set Gravity Override(Gravity Override=@n14)
   n15 = Get Dynamic Material Instance(execute=[@n39.then, @n2.then_4], SkeletalMeshComponent=GetPartSkeletalMeshComponent())
   n16 = PickaxeItemDefinitionHasAnyMetaTag(Tags=Make Array("[0]"=(TagName="Cosmetics.MetaTags.ApplePound")), bRequiresEquipped=true)
   > n17 = Branch(Condition=@n16)
   > n18 = Branch(Condition=Get bIsSkyDiving())
   SetScalarParameterValue(execute=[@n37.then, @n38.else, @n18.then, @n17.else], self=@n15, ParameterName=Hide Element 01, Value=0)
   n19 = GetPartSkeletalMeshComponent()
   > SetScalarParameterValueOnOverlayMaterial!(execute, MeshComponent:Mesh Component Object Reference=@n19, ParameterName:Name=Hide Element 01, ParameterValue:float=0) -> (then)
   SetScalarParameterValue(execute=[@n38.then, @n18.else], self=@n15, ParameterName=Hide Element 01, Value=1)
   > SetScalarParameterValueOnOverlayMaterial(MeshComponent=@n19, ParameterName=Hide Element 01, ParameterValue=1)
   n20 = Get bIsSkydiveDiveMode()
   Set SkyDivingModeAlpha(execute=@n2.then_7, SkyDivingModeAlpha:double=SelectFloat(A=0, B=1, bPickA=@n20)) -> (Output_Get:double)
   > Set Updates Backpack Crouch Alpha(Updates Backpack Crouch Alpha:double=SelectFloat(A=0, B=1, bPickA=Get bIsCrouching())) -> (Output_Get:double)
   > n21 = Branch(Condition=Get bIsSkyDiving())
   > Do Once(Reset=@n22.then)
   > Set Start Skydiving(Start Skydiving:double=1) -> (Output_Get:double)
   > Delay(Duration=1.2)
   > Set Start Skydiving(Start Skydiving=0)
   n22 = Set Start Skydiving(execute=@n21.else, Start Skydiving=0)
   n23 = Branch(execute=@n2.then_8, Condition=Get bIsPlayingMeleeAnim())
   > Do Once(Reset=@n24.then)
   > Set NewVar(NewVar:double=1) -> (Output_Get:double)
   n24 = Set NewVar(execute=@n23.else, NewVar=0)
   Branch(execute=@n2.then_5, Condition=Get bIsInFrontEnd())
   n25 = Make Array("[0]"=(TagName="Cosmetics.MetaTags.ApplePound"))
   n26 = GetCurveValue(self=GetOwnerFrontEndAnimInstance(), CurveName=FrontEnd_Idle_Bypass)
   > n27 = Branch(Condition=AND Boolean(A=CharacterItemDefinitionHasAnyMetaTag(Tags=@n25), B="float >= float"(A=@n26, B=1)))
   > Set FE Idle Active(FE Idle Active:bool=true) -> (Output_Get:bool)
   Set FE Idle Active(execute=@n27.else, FE Idle Active=false)
   n28 = BreakTransform(InTransform=GetSocketTransform(self=Get BodyMesh() -> (BodyMesh:Skeletal Mesh Component Object Reference), InSocketName=upperarm_l, TransformSpace=RTS_World))
   Set upperarm_l translation(execute=@n2.then_1, upperarm_l translation:Vector=@n28) -> (Output_Get:Vector)
   > Set upperarm_l rotation(upperarm_l rotation:Rotator=@n28.Rotation) -> (Output_Get:Rotator)
   n29 = BreakTransform(InTransform=GetSocketTransform(self=Get BodyMesh(), InSocketName=pelvis, TransformSpace=RTS_World))
   > Set Location_pelvis(Location_pelvis:Vector=@n29) -> (Output_Get:Vector)
   > Set Rotation_Pelvis(Rotation_Pelvis:Rotator=@n29.Rotation) -> (Output_Get:Rotator)
   > Set!(
    execute
    Rotation_skyfloating_CLOSED_Roll:float=SelectFloat(A=0, B=-5, bPickA=Get bIsParachuteOpened())
    Rotation_skyfloating_CLOSED_Pitch:float=SelectFloat(A=0, B=-12, bPickA=Get bIsParachuteOpened())
    Rotation_skyfloating_CLOSED_Yaw:float=SelectFloat(A=0, B=-10, bPickA=Get bIsParachuteOpened())
   ) -> (then, Output_Get:Rotator)
   > Set!(
    execute
    spine_skyfloating_X:double=0
    spine_skyfloating_Y:double=0
    spine_skyfloating_Z:double=SelectFloat(A=0, B=-3, bPickA=Get bIsParachuteOpened())
   ) -> (then, Output_Get:Vector)
   n30 = GetOwnerSkeletalMeshComponent()
   n31 = BreakRotator(InRot=BreakTransform(InTransform=GetDeltaTransformFromRefPose(self=@n30, BoneName=spine_05, BaseName=spine_01)).Rotation)
   n32 = MapRangeClamped(Value=@n31.Pitch, InRangeA=30, InRangeB=-30, OutRangeA=-30, OutRangeB=30)
   n33 = MapRangeClamped(Value=@n31.Pitch, InRangeA=0, InRangeB=-30, OutRangeA=-5, OutRangeB=25)
   n34 = OR Boolean(A=Get bIsSkydiveFloating(), B=Get bIsSkyDiving())
   Set backpack_rotate(
    execute=@n2.then_6
    backpack_rotate=MakeRotator(Roll=MapRangeClamped(Value=@n31.Yaw, InRangeA=0, InRangeB=-30, OutRangeA=5, OutRangeB=25), Pitch=SelectFloat(A=@n33, B=@n32, bPickA=@n34))
   )
   > Set backpack_translation(backpack_translation=MakeVector(Y=MapRangeClamped(Value=@n31.Yaw, InRangeA=0, InRangeB=-30, OutRangeB=-6.5)))
   n35 = GetOwnerSkeletalMeshComponent()
   n36 = BreakRotator(InRot=BreakTransform(InTransform=GetDeltaTransformFromRefPose(self=@n35, BoneName=spine_05, BaseName=spine_01)).Rotation)
   Set opening location(execute=@n2.then_0, opening location:Vector=MakeVector(Y=MapRangeClamped(Value=@n36.Yaw, InRangeA=0, InRangeB=-40, OutRangeB=-30))) -> (Output_Get:Vector)
   > Set closing location(closing location:Vector=MakeVector(Y=MapRangeClamped(Value=@n36.Yaw, InRangeA=0, InRangeB=-40, OutRangeB=-40))) -> (Output_Get:Vector)
   > Set sky floating(sky floating:Vector=MakeVector(Y=MapRangeClamped(Value=@n36.Yaw, InRangeA=0, InRangeB=-20, OutRangeB=-4))) -> (Output_Get:Vector)
   > Set spine_reverse(spine_reverse:Vector=MakeVector(Y=MapRangeClamped(Value=@n36.Yaw, InRangeA=0, InRangeB=20, OutRangeB=-30))) -> (Output_Get:Vector)
   > Set!(execute, moving left_X:double=0, moving left_Y:double=SelectFloat(A=0, B=0, bPickA=Get bIsParachuteOpened()), moving left_Z:double=0) -> (then, Output_Get)
   > Set spine_interp(spine_interp:Vector=VInterpTo!(
     Current:Vector=Get spine_reverse() -> (spine_reverse:Vector)
     Target:Vector=Get opening location() -> (opening location:Vector)
     DeltaTime:float=Get Delta Time X()
     InterpSpeed:float=0.3
    ) -> (ReturnValue:Vector)) -> (Output_Get:Vector)
   n37 = Event OnBeginSkydiving()
   Event OnEndSkydiving()
   > n38 = Branch(Condition=@n16)
   n39 = Event OnWeaponChanged()
   n40 = Event OnCharacterCustomizationLoaded()
   > n41 = Sequence!(execute) -> (then_0, then_1, then_2:Exec)
   Set BodyMesh(execute=@n41.then_1, BodyMesh=Get Array Item with EFort Custom Part Type(Array=@n40.CustomMeshParts))
   OR Boolean(A="Equal (Float)"(A=Get Start Skydiving()), B=@n20)
  ]}
]
variables [
 backpack_rotate {type=Rotator}
 backpack_translation {type=Vector}
 BodyMesh {type=SkeletalMeshComponent}
 closing location {type=Vector}
 Delta Time X {type=number}
 End OP {type=number}
 FE Idle Active {type=boolean}
 Gravity Override {type=Vector}
 Gravity Override Params {type=GravityOverrideParamsStruct}
 Is Sky Diving {type=boolean}
 LeftMovementAlphaSkydiving {type=number}
 LeftMovementSkydiving {type=boolean}
 Location_pelvis {type=Vector}
 moving left {type=Vector}
 NewVar {type=number}
 opening location {type=Vector}
 parachute closed {type=boolean}
 parachute closing {type=boolean}
 parachute opening {type=boolean}
 RightMovementAlphaSkydiving {type=number}
 RightMovementSkydiving {type=boolean}
 Rotation_Pelvis {type=Rotator}
 Rotation_skyfloating_CLOSED {type=Rotator}
 Sky corr {type=boolean}
 sky floating {type=Vector}
 SkyDivingModeAlpha {type=number}
 spine_interp {type=Vector}
 spine_reverse {type=Vector}
 spine_skyfloating {type=Vector}
 Start OP {type=number}
 Start Skydiving {type=number}
 Updates Backpack Crouch Alpha {type=number}
 upperarm_l rotation {type=Rotator}
 upperarm_l translation {type=Vector}
]
